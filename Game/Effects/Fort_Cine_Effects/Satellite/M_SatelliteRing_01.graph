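# Translucent, unlit, two-sided mesh-particle material for a satellite ring effect in the cinematic effects set.
# A Fresnel function blends outer and inner blue color parameters, tinted by particle color and passed through an emissive-adjustment function.
# Particle alpha is multiplied by vertex alpha, and a swirl-noise normal map pans at a dynamic-parameter speed.
# The dynamic parameter also names refraction and tiling channels, but no node shown reads them.
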Material /Game/Effects/Fort_Cine_Effects/Satellite/M_SatelliteRing_01
graph {
 properties {
  bUsedWithMeshParticles=true
  blendMode=BLEND_Translucent
  pixelDepthOffsetMode=PDOM_Legacy
  previewMesh=&/S_Ring_02
  shadingModel=MSM_Unlit
  twoSided=true
  usedShadingModels=Unlit
 }
 nodes [
  n1 = VectorParameter() {parameterName=Color Outer, defaultValue=<rgba 0.5116 2.9244 8 1>}
  n2 = VectorParameter() {parameterName=Color Inner, defaultValue=<rgba 0.0125 0.1869 0.5 1>}
  n3 = MaterialFunctionCall() {materialFunction=&MF_Fresnel_FX}
  n4 = ParticleColor()
  MaterialFunctionCall(Emissive=Multiply(A=LinearInterpolate(A=@n1.RGB, B=@n2.RGB, Alpha=@n3.Fresnel), B=@n4.RGB)) {materialFunction=&MF_FX_EmissiveAdjustments}
  Multiply(A=@n4.A, B=VertexColor().A)
  n5 = Time()
  n6 = DynamicParameter() {defaultValue=<rgba 1.2 1 4 0.5>, paramNames [Refraction Amount, Panner Speed, U Tile, V Tile]}
  TextureSample(UVs=Panner(Coordinate=TextureCoordinate() {uTiling=10, vTiling=0.3}, Time=Multiply(A=@n5, B=@n6.Panner Speed)) {bFractionalPart=true, speedX=0.8, speedY=-3.2}) {samplerType=SAMPLERTYPE_Normal, texture=/Game/Global/Textures/Default/Noise/T_SwirlNoise_N}
 ]
}
